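# Control Rig for the female RankedTrooper Noble face accessory mesh that adjusts the back hair as the head moves.
# The left, center and right back-hair master bones are rotation-constrained equally to spine_05 and head.
# Three spherical pose readers track head orientation against neck bones, weighting corrective rotations on each chain's first bone.
# One reader rotates all three chains equally; the other two rotate the center slightly and one opposite side more.
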ControlRigBlueprint /BRCosmetics/Characters/Player/Female/Medium/Bodies/F_MED_RankedTrooper_Noble/Meshes/Parts/F_MED_RankedTrooper_Noble_FaceAcc_CtrlRig
graphs [
 RigVMFunctionLibrary {links [], nodes []}
 RigVMModel {nodes [
   RigUnit_BeginExecution = &RigUnit_BeginExecution()
   &RigUnit_RotationConstraintLocalSpaceOffset(
    AdvancedSettings=(RotationOrderForFilter=XZY)
    Child=(Type=Bone,Name="dyn_Hair_bk_l_MSTR")
    Filter=(bX=True,bY=True,bZ=True)
    Parents=((Item=(Type=Bone,Name="spine_05"),Weight=0.5),(Item=(Type=Bone,Name="head"),Weight=0.5))
    Weight=1
    ExecutePin=@RigUnit_BeginExecution.ExecutePin
   )
   > &RigUnit_RotationConstraintLocalSpaceOffset(
    AdvancedSettings=(RotationOrderForFilter=XZY)
    Child=(Type=Bone,Name="dyn_Hair_bk_MSTR")
    Filter=(bX=True,bY=True,bZ=True)
    Parents=((Item=(Type=Bone,Name="spine_05"),Weight=0.5),(Item=(Type=Bone,Name="head"),Weight=0.5))
    Weight=1
   )
   > &RigUnit_RotationConstraintLocalSpaceOffset(
    AdvancedSettings=(RotationOrderForFilter=XZY)
    Child=(Type=Bone,Name="dyn_Hair_bk_r_MSTR")
    Filter=(bX=True,bY=True,bZ=True)
    Parents=((Item=(Type=Bone,Name="spine_05"),Weight=0.5),(Item=(Type=Bone,Name="head"),Weight=0.5))
    Weight=1
    bMaintainOffset=true
   )
   > n1 = &RigUnit_SphericalPoseReader(
    ActiveRegionScaleFactors=(PositiveWidth=1,NegativeWidth=1,PositiveHeight=1,NegativeHeight=1)
    ActiveRegionSize=0.1
    Debug=(bDrawDebug=false,DebugScale=25,DebugSegments=20,DebugThickness=0.25)
    DriverAxis=(X=0,Y=-1,Z=0)
    DriverItem=(Type=Bone,Name="head")
    FalloffRegionScaleFactors=(PositiveWidth=2,NegativeWidth=2,PositiveHeight=2,NegativeHeight=2)
    FalloffSize=0.2
    FlipWidthScaling=false
    OptionalParentItem=(Type=Bone,Name="neck_01")
    RotationOffset=(X=-90,Y=0,Z=-55)
   )
   > &RigUnit_ModifyTransforms(
    ItemToModify=((Item=(Type=Bone,Name="dyn_Hair_bk_01_ct"),Transform=(Rotation=(X=0,Y=-0.1305,Z=0,W=0.9914),Translation=(X=0,Y=0,Z=0),Scale3D=(X=1,Y=1,Z=1))))
    WeightMinimum=0
    Weight=@n1.OutputParam
   )
   > &RigUnit_ModifyTransforms(
    ItemToModify=((Item=(Type=Bone,Name="dyn_Hair_bk_01_l"),Transform=(Rotation=(X=0,Y=-0.1305,Z=0,W=0.9914),Translation=(X=0,Y=0,Z=0),Scale3D=(X=1,Y=1,Z=1))))
    WeightMinimum=0
    Weight=@n1.OutputParam
   )
   > &RigUnit_ModifyTransforms(
    ItemToModify=((Item=(Type=Bone,Name="dyn_Hair_bk_01_r"),Transform=(Rotation=(X=0,Y=-0.1305,Z=0,W=0.9914),Translation=(X=0,Y=0,Z=0),Scale3D=(X=1,Y=1,Z=1))))
    WeightMinimum=0
    Weight=@n1.OutputParam
   )
   > n2 = &RigUnit_SphericalPoseReader(
    ActiveRegionScaleFactors=(PositiveWidth=1,NegativeWidth=0.5,PositiveHeight=1,NegativeHeight=1)
    ActiveRegionSize=0.1
    Debug=(bDrawDebug=false,DebugScale=25,DebugSegments=20,DebugThickness=0.25)
    DriverAxis=(X=0,Y=-1,Z=0)
    DriverItem=(Type=Bone,Name="head")
    FalloffRegionScaleFactors=(PositiveWidth=1,NegativeWidth=0.5,PositiveHeight=1,NegativeHeight=1)
    FalloffSize=0.2
    OptionalParentItem=(Type=Bone,Name="neck_02")
    RotationOffset=(X=-55,Y=0,Z=-10)
   )
   > &RigUnit_ModifyTransforms(
    ItemToModify=((Item=(Type=Bone,Name="dyn_Hair_bk_01_ct"),Transform=(Rotation=(X=0,Y=-0.0436,Z=0,W=0.999),Translation=(X=0,Y=0,Z=0),Scale3D=(X=1,Y=1,Z=1))))
    WeightMinimum=0
    Weight=@n2.OutputParam
   )
   > &RigUnit_ModifyTransforms(
    ItemToModify=((Item=(Type=Bone,Name="dyn_Hair_bk_01_l"),Transform=(Rotation=(X=0,Y=0,Z=0,W=1),Translation=(X=0,Y=0,Z=0),Scale3D=(X=1,Y=1,Z=1))))
    WeightMinimum=0
    Weight=@n2.OutputParam
   )
   > &RigUnit_ModifyTransforms(
    ItemToModify=((Item=(Type=Bone,Name="dyn_Hair_bk_01_r"),Transform=(Rotation=(X=0,Y=-0.1305,Z=0,W=0.9914),Translation=(X=0,Y=0,Z=0),Scale3D=(X=1,Y=1,Z=1))))
    WeightMinimum=0
    Weight=@n2.OutputParam
   )
   > n3 = &RigUnit_SphericalPoseReader(
    ActiveRegionScaleFactors=(PositiveWidth=1,NegativeWidth=0.5,PositiveHeight=1,NegativeHeight=1)
    ActiveRegionSize=0.1
    Debug=(bDrawDebug=false,DebugScale=25,DebugSegments=20,DebugThickness=0.25)
    DriverAxis=(X=0,Y=-1,Z=0)
    DriverItem=(Type=Bone,Name="head")
    FalloffRegionScaleFactors=(PositiveWidth=1,NegativeWidth=0.5,PositiveHeight=1,NegativeHeight=1)
    FalloffSize=0.2
    OptionalParentItem=(Type=Bone,Name="neck_02")
    RotationOffset=(X=235,Y=0,Z=-10)
   )
   > &RigUnit_ModifyTransforms(
    ItemToModify=((Item=(Type=Bone,Name="dyn_Hair_bk_01_ct"),Transform=(Rotation=(X=0,Y=-0.0436,Z=0,W=0.999),Translation=(X=0,Y=0,Z=0),Scale3D=(X=1,Y=1,Z=1))))
    WeightMinimum=0
    Weight=@n3.OutputParam
   )
   > &RigUnit_ModifyTransforms(
    ItemToModify=((Item=(Type=Bone,Name="dyn_Hair_bk_01_l"),Transform=(Rotation=(X=0,Y=-0.1305,Z=0,W=0.9914),Translation=(X=0,Y=0,Z=0),Scale3D=(X=1,Y=1,Z=1))))
    WeightMinimum=0
    Weight=@n3.OutputParam
   )
   > &RigUnit_ModifyTransforms(
    ItemToModify=((Item=(Type=Bone,Name="dyn_Hair_bk_01_r"),Transform=(Rotation=(X=0,Y=0,Z=0,W=1),Translation=(X=0,Y=0,Z=0),Scale3D=(X=1,Y=1,Z=1))))
    WeightMinimum=0
    Weight=@n3.OutputParam
   )
  ]}
]
hierarchy [
 root {loc=<0 0 0>, rot=<0 0 0 1>}
 pelvis {parent=root, loc=<0 -0.5301 98.0571>, rot=<0.0127 0.707 -0.0127 -0.707>}
 spine_01 {parent=pelvis, loc=<6.1208 -0.9328 0>, rot=<0 0 0.034 -0.9994>}
 spine_02 {parent=spine_01, loc=<6.5997 -0.2095 0>, rot=<0 0 0.0238 -0.9997>}
 spine_03 {parent=spine_02, loc=<6.4914 -0.1237 0>, rot=<0 0 0.0149 -0.9999>}
 spine_04 {parent=spine_03, loc=<6.9598 0.0794 0>, rot=<0 0 -0.0132 -0.9999>}
 spine_05 {parent=spine_04, loc=<6.9987 0 0>, rot=<0 0 -0.0015 -1>}
 neck_01 {parent=spine_05, loc=<11.3138 0.1529 0>, rot=<0 0 -0.2003 -0.9797>}
 neck_02 {parent=neck_01, loc=<4.5627 0 0>, rot=<0 0 0.0504 -0.9987>}
 head {parent=neck_02, loc=<5.3914 0 0>, rot=<0 0 0.115 -0.9934>}
 dyn_L_HairBang_MSTR {parent=head, loc=<9.5838 6.6511 -4.2632>, rot=<0.0854 -0.9687 -0.1281 -0.1945>}
 dyn_L_HairBang_01_jnt {parent=dyn_L_HairBang_MSTR, loc=<0 0 0>, rot=<0 0 0 1>}
 dyn_L_HairBang_02_jnt {parent=dyn_L_HairBang_01_jnt, loc=<3.8574 0 0>, rot=<0.0407 -0.1746 -0.0072 -0.9838>}
 dyn_L_HairBang_03_jnt {parent=dyn_L_HairBang_02_jnt, loc=<3.3875 0 0>, rot=<0.0113 -0.0792 -0.1403 -0.9869>}
 dyn_L_HairTip_01_jnt {parent=dyn_L_HairBang_03_jnt, loc=<3.8574 0 0>, rot=<0.0163 -0.1097 -0.1461 -0.983>}
 dyn_L_HairTip_02_jnt {parent=dyn_L_HairBang_03_jnt, loc=<2.4144 1.23 0.0898>, rot=<0.1602 -0.0607 -0.391 -0.9043>}
 dyn_R_HairBang_MSTR {parent=head, loc=<9.6161 6.444 4.2762>, rot=<0.2169 0.1263 0.9625 -0.1026>}
 dyn_R_HairBang_01_jnt {parent=dyn_R_HairBang_MSTR, loc=<0 0 0>, rot=<0 0 0 1>}
 dyn_R_HairBang_02_jnt {parent=dyn_R_HairBang_01_jnt, loc=<3.8574 0 0>, rot=<-0.0882 -0.1644 0.144 -0.9718>}
 dyn_R_HairBang_03_jnt {parent=dyn_R_HairBang_02_jnt, loc=<3.3875 0 0>, rot=<-0.0772 -0.0801 0.2423 -0.9638>}
 dyn_R_HairBang_04_jnt {parent=dyn_R_HairBang_03_jnt, loc=<2.6339 0 0.0001>, rot=<-0.0911 -0.187 0.3138 -0.9264>}
 dyn_R_SmallHairBang_MSTR {parent=head, loc=<8.8628 8.3315 3.9254>, rot=<0.2009 0.256 0.9364 -0.1311>}
 dyn_R_SmallHairBang_01_jnt {parent=dyn_R_SmallHairBang_MSTR, loc=<0 0 0>, rot=<0 0 0 1>}
 dyn_R_SmallHairBang_02_jnt {parent=dyn_R_SmallHairBang_01_jnt, loc=<1.4447 0 0>, rot=<0.0553 -0.1185 0.1368 -0.9819>}
 dyn_R_SmallHairBang_03_jnt {parent=dyn_R_SmallHairBang_02_jnt, loc=<1.6206 0 0>, rot=<0.0157 -0.1292 0.2283 -0.9649>}
 dyn_R_SmallHairBang_04_jnt {parent=dyn_R_SmallHairBang_03_jnt, loc=<1.6819 0 0.0001>, rot=<-0.0125 -0.1125 0.11 -0.9875>}
 dyn_L_Hairstrand_MSTR {parent=head, loc=<10.4438 2.8653 -4.8795>, rot=<0.0822 -0.9566 -0.039 -0.2768>}
 dyn_L_Hairstrand_01_jnt {parent=dyn_L_Hairstrand_MSTR, loc=<0 0 0>, rot=<0 -0.0333 0 -0.9994>}
 dyn_L_Hairstrand_02_jnt {parent=dyn_L_Hairstrand_01_jnt, loc=<5.0954 0 0>, rot=<0 -0.2256 0 -0.9742>}
 dyn_L_Hairstrand_03_jnt {parent=dyn_L_Hairstrand_02_jnt, loc=<8.0123 0 0>, rot=<0.072 0.0471 -0.0603 -0.9945>}
 dyn_L_Hairstrand_01_Tip_01_jnt {parent=dyn_L_Hairstrand_03_jnt, loc=<6.2036 0 0>, rot=<0.141 0.028 -0.2325 -0.9619>}
 dyn_L_Hairstrand_01_Tip_02_jnt {parent=dyn_L_Hairstrand_01_Tip_01_jnt, loc=<5.0655 0 0>, rot=<0.0601 0.054 -0.2288 -0.9701>}
 dyn_L_Hairstrand_02_Tip_01_jnt {parent=dyn_L_Hairstrand_03_jnt, loc=<5.2412 2.0769 -1.222>, rot=<0.2722 -0.1753 -0.2801 -0.9037>}
 dyn_L_Hairstrand_02_Tip_02_jnt {parent=dyn_L_Hairstrand_02_Tip_01_jnt, loc=<2.6219 0 0>, rot=<0.0387 0.0088 -0.1908 -0.9808>}
 dyn_R_Hairstrand_MSTR {parent=head, loc=<10.4434 2.8653 4.8795>, rot=<0.2768 0.039 0.9566 -0.0822>}
 dyn_R_Hairstrand_01_jnt {parent=dyn_R_Hairstrand_MSTR, loc=<0 0 0>, rot=<0 -0.0333 0 -0.9994>}
 dyn_R_Hairstrand_02_jnt {parent=dyn_R_Hairstrand_01_jnt, loc=<5.0955 0 0>, rot=<0 -0.2256 0 -0.9742>}
 dyn_R_Hairstrand_03_jnt {parent=dyn_R_Hairstrand_02_jnt, loc=<8.0115 0.0001 0>, rot=<-0.2352 -0.0039 0.1321 -0.9629>}
 dyn_R_Hairstrand_01_Tip_01_jnt {parent=dyn_R_Hairstrand_03_jnt, loc=<6.2042 0 -0.0001>, rot=<-0.0068 -0.0044 0.2342 -0.9722>}
 dyn_R_Hairstrand_01_Tip_02_jnt {parent=dyn_R_Hairstrand_01_Tip_01_jnt, loc=<5.065 -0.0002 0.0001>, rot=<-0.0387 -0.0637 0.218 -0.9731>}
 dyn_R_Hairstrand_01_Tip_03_jnt {parent=dyn_R_Hairstrand_01_Tip_02_jnt, loc=<3.9941 0.0003 -0.0002>, rot=<0.1182 0.044 -0.0772 -0.989>}
 dyn_R_Hairstrand_02_Tip_01_jnt {parent=dyn_R_Hairstrand_03_jnt, loc=<5.8584 -1.1009 -0.9548>, rot=<-0.1248 -0.1023 0.2117 -0.9639>}
 dyn_R_Hairstrand_02_Tip_02_jnt {parent=dyn_R_Hairstrand_02_Tip_01_jnt, loc=<2.7724 0 0>, rot=<-0.0575 -0.1102 0.2171 -0.9682>}
 dyn_R_Hairstrand_02_Tip_03_jnt {parent=dyn_R_Hairstrand_02_Tip_02_jnt, loc=<3.0102 -0.0003 0>, rot=<0.0092 -0.1591 0.0741 -0.9844>}
 dyn_Hair_bk_MSTR {parent=head, loc=<9.8456 -6.1269 0>, rot=<0.5 -0.5 0.5 -0.5>}
 dyn_Hair_bk_01_ct {parent=dyn_Hair_bk_MSTR, loc=<0 0 0>, rot=<0 -0.5708 0 -0.8211>}
 dyn_Hair_bk_02_ct {parent=dyn_Hair_bk_01_ct, loc=<7.522 0 0>, rot=<0 -0.1108 0 -0.9938>}
 dyn_Hair_bk_03_ct {parent=dyn_Hair_bk_02_ct, loc=<7.522 0 0>, rot=<0 0.077 0 -0.997>}
 dyn_Hair_bk_04_ct {parent=dyn_Hair_bk_03_ct, loc=<7.522 0 0>, rot=<0 -0.0034 0 -1>}
 dyn_Hair_bk_05_ct {parent=dyn_Hair_bk_04_ct, loc=<7.522 0 0>, rot=<0 -0.0975 0 -0.9952>}
 dyn_Hair_bk_06_ct {parent=dyn_Hair_bk_05_ct, loc=<7.522 0 0>, rot=<0 0 0 1>}
 dyn_Hair_bk_l_MSTR {parent=head, loc=<9.6609 -1.634 -6.4062>, rot=<0.5 -0.5 0.5 -0.5>}
 dyn_Hair_bk_01_l {parent=dyn_Hair_bk_l_MSTR, loc=<0 0 0>, rot=<0 -0.4882 0 -0.8727>}
 dyn_Hair_bk_02_l {parent=dyn_Hair_bk_01_l, loc=<7.522 0 0>, rot=<0 -0.16 0 -0.9871>}
 dyn_Hair_bk_03_l {parent=dyn_Hair_bk_02_l, loc=<7.522 0 0>, rot=<0 0.1298 0 -0.9915>}
 dyn_Hair_bk_04_l {parent=dyn_Hair_bk_03_l, loc=<7.522 0 0>, rot=<0 -0.1239 0 -0.9923>}
 dyn_Hair_bk_05_l {parent=dyn_Hair_bk_04_l, loc=<7.522 0 0>, rot=<0 -0.1201 0 -0.9928>}
 dyn_Hair_bk_06_l {parent=dyn_Hair_bk_05_l, loc=<7.522 0 0>, rot=<0 0 0 1>}
 dyn_Hair_bk_r_MSTR {parent=head, loc=<9.6609 -1.634 6.406>, rot=<0.5 -0.5 0.5 -0.5>}
 dyn_Hair_bk_01_r {parent=dyn_Hair_bk_r_MSTR, loc=<0 0 0>, rot=<0 -0.4882 0 -0.8727>}
 dyn_Hair_bk_02_r {parent=dyn_Hair_bk_01_r, loc=<7.522 0 0>, rot=<0 -0.16 0 -0.9871>}
 dyn_Hair_bk_03_r {parent=dyn_Hair_bk_02_r, loc=<7.522 0 0>, rot=<0 0.1298 0 -0.9915>}
 dyn_Hair_bk_04_r {parent=dyn_Hair_bk_03_r, loc=<7.522 0 0>, rot=<0 -0.1239 0 -0.9923>}
 dyn_Hair_bk_05_r {parent=dyn_Hair_bk_04_r, loc=<7.522 0 0>, rot=<0 -0.1201 0 -0.9928>}
 dyn_Hair_bk_06_r {parent=dyn_Hair_bk_05_r, loc=<7.522 0 0>, rot=<0 0 0 -1>}
 clavicle_l {parent=spine_05, loc=<5.4831 -0.1053 -3.2946>, rot=<0.0465 -0.7289 -0.009 -0.6829>}
 upperarm_l {parent=clavicle_l, loc=<9.8658 -1.2136 -0.4105>, rot=<0.0164 -0.3245 -0.0269 -0.9454>}
 upperarm_twist_01_l {parent=upperarm_l, loc=<7.2135 -0.1134 0>, rot=<0 0 0 1>}
 deform_upperarm_bk_base_l {parent=upperarm_twist_01_l, loc=<-6.4053 -4.3746 0.314>, rot=<0 0 0 -1>}
 deform_upperarm_bk_l {parent=deform_upperarm_bk_base_l, loc=<0 0 0>, rot=<0 0 0 1>}
 deform_upperarm_fr_base_l {parent=upperarm_twist_01_l, loc=<-5.6312 5.092 0.3354>, rot=<0 0 0 -1>}
 deform_upperarm_fr_l {parent=deform_upperarm_fr_base_l, loc=<0 0 0>, rot=<0 0 0 1>}
 deform_upperarm_up_base_l {parent=upperarm_twist_01_l, loc=<-8.891 0.4825 3.9517>, rot=<0 0 0 -1>}
 deform_upperarm_up_l {parent=deform_upperarm_up_base_l, loc=<0 0 0>, rot=<0 0 0 1>}
 clavicle_r {parent=spine_05, loc=<5.4831 -0.1053 3.2946>, rot=<0.7289 0.0465 -0.6829 0.009>}
 upperarm_r {parent=clavicle_r, loc=<-9.8658 1.2136 0.4105>, rot=<0.0164 -0.3245 -0.0269 -0.9454>}
 upperarm_twist_01_r {parent=upperarm_r, loc=<-7.2135 0.1134 0>, rot=<0 0 0 -1>}
 deform_upperarm_bk_base_r {parent=upperarm_twist_01_r, loc=<6.4053 4.3746 -0.314>, rot=<0 0 0 -1>}
 deform_upperarm_bk_r {parent=deform_upperarm_bk_base_r, loc=<0 0 0>, rot=<0 0 0 1>}
 deform_upperarm_fr_base_r {parent=upperarm_twist_01_r, loc=<5.6312 -5.092 -0.3354>, rot=<0 0 0 -1>}
 deform_upperarm_fr_r {parent=deform_upperarm_fr_base_r, loc=<0 0 0>, rot=<0 0 0 1>}
 deform_upperarm_up_base_r {parent=upperarm_twist_01_r, loc=<8.891 -0.4825 -3.9517>, rot=<0 0 0 -1>}
 deform_upperarm_up_r {parent=deform_upperarm_up_base_r, loc=<0 0 0>, rot=<0 0 0 1>}
 upperarm_twist_02_r {parent=upperarm_r, loc=<-19.7998 0.1134 0>, rot=<0 0 0 -1>}
 deform_pec_aim_base_l {parent=spine_04, loc=<3.2456 9.3309 -1.2237>, rot=<0.0603 -0.4033 0.3318 -0.8507>}
 deform_pec_aim_l {parent=deform_pec_aim_base_l, loc=<0 0 0>, rot=<0 0 0 1>}
 deform_pec_base_l {parent=deform_pec_aim_l, loc=<9.3665 0 0>, rot=<0 0 0 1>}
 deform_pec_l {parent=deform_pec_base_l, loc=<0 0 0>, rot=<0 0 0 1>}
 deform_pec_aim_base_r {parent=spine_04, loc=<3.2452 9.3309 1.2237>, rot=<-0.4033 -0.0603 0.8507 0.3318>}
 deform_pec_aim_r {parent=deform_pec_aim_base_r, loc=<0 0 0>, rot=<0 0 0 1>}
 deform_pec_base_r {parent=deform_pec_aim_r, loc=<-9.3667 -0.0003 -0.0004>, rot=<0 0 0 1>}
 deform_pec_r {parent=deform_pec_base_r, loc=<0 0 0>, rot=<0 0 0 1>}
 thigh_r {parent=pelvis, loc=<-1.0855 -0.2109 8.7053>, rot=<0.0355 -0.0009 0.9993 0.0155>}
 thigh_twist_01_r {parent=thigh_r, loc=<27.952 0 0>, rot=<0.0013 0 0 -1>}
 deform_glute_base_r {parent=pelvis, loc=<-7.3464 -9.0213 9.1245>, rot=<0.031 -0.0172 0.8935 -0.4477>}
 deform_glute_r {parent=deform_glute_base_r, loc=<0 0 0>, rot=<0 0 0 1>}
]
