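# Physics asset for the Hairpin mesh part of the female medium RankedTrooper Noble body, providing capsule bodies and joints for its hair.
# Capsules on pelvis, spine, neck, head, clavicles and upper arms form the body volume, linked by free angular constraints.
# Bang, strand and back-hair chains hang from zero-size MSTR bodies on the head.
# Bangs and strands, some forking into tips, have tight swing and twist limits, while the three back-hair chains rotate freely.
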PhysicsAsset /BRCosmetics/Characters/Player/Female/Medium/Bodies/F_MED_RankedTrooper_Noble/Meshes/Parts/F_MED_RankedTrooper_Noble_Hairpin_Physics
bodies {
 clavicle_l [{shapeType=Capsule, length=12, radius=5.5, shapeName=clavicle_l_capsule, center=<3.5 -1 -1>, rotation=<rot 80 -0.0003 -0.0003>}]
 clavicle_r [{shapeType=Capsule, length=12, radius=5.5, shapeName=clavicle_r_capsule, center=<-3.5 1 1>, rotation=<rot 80 -0.0003 -0.0004>}]
 dyn_Hair_bk_01_ct [{shapeType=Capsule, length=6, radius=6, shapeName=dyn_Hair_bk_01_ct_capsule, center=<8 0 3.5>, rotation=<rot 0 90 90>}]
 dyn_Hair_bk_01_l [{shapeType=Capsule, length=6, radius=6, shapeName=dyn_Hair_bk_01_l_capsule, center=<8 0 3.5>, rotation=<rot 0 90 90>}]
 dyn_Hair_bk_01_r [{shapeType=Capsule, length=6, radius=6, shapeName=dyn_Hair_bk_01_r_capsule, center=<8 0 3.5>, rotation=<rot 0 90 90>}]
 dyn_Hair_bk_02_ct [{shapeType=Capsule, length=6, radius=6, shapeName=dyn_Hair_bk_02_ct_capsule, center=<8 0 3.5>, rotation=<rot 0 90 90>}]
 dyn_Hair_bk_02_l [
  {shapeType=Capsule, length=6, radius=6, shapeName=dyn_Hair_bk_02_l_capsule, center=<8 0 3.5>, rotation=<rot 0 90 90>}
  {shapeType=Capsule, length=6, radius=5, shapeName=dyn_Hair_bk_02_l_capsule, center=<7.8112 2.9417 0.5565>, rotation=<rot 0 100 89.9997>}
 ]
 dyn_Hair_bk_02_r [
  {shapeType=Capsule, length=6, radius=6, shapeName=dyn_Hair_bk_02_r_capsule, center=<8 0 3.5>, rotation=<rot 0 90 90>}
  {shapeType=Capsule, length=6, radius=5, shapeName=dyn_Hair_bk_02_r_capsule, center=<8.3095 -2.8638 0.7576>, rotation=<rot 0 70 89.9999>}
 ]
 dyn_Hair_bk_03_ct [{shapeType=Capsule, length=6, radius=6, shapeName=dyn_Hair_bk_03_ct_capsule, center=<8 0 3.5>, rotation=<rot 0 90 90>}]
 dyn_Hair_bk_03_l [{shapeType=Capsule, length=6, radius=6, shapeName=dyn_Hair_bk_03_l_capsule, center=<8 0 3.5>, rotation=<rot 0 90 90>}]
 dyn_Hair_bk_03_r [{shapeType=Capsule, length=6, radius=6, shapeName=dyn_Hair_bk_03_r_capsule, center=<8 0 3.5>, rotation=<rot 0 90 90>}]
 dyn_Hair_bk_MSTR [{shapeType=Capsule, length=0, radius=0, shapeName=dyn_Hair_bk_MSTR_capsule}]
 dyn_Hair_bk_l_MSTR [{shapeType=Capsule, length=0, radius=0, shapeName=dyn_Hair_bk_l_MSTR_capsule}]
 dyn_Hair_bk_r_MSTR [{shapeType=Capsule, length=0, radius=0, shapeName=dyn_Hair_bk_r_MSTR_capsule}]
 dyn_L_HairBang_01_jnt [{shapeType=Capsule, length=2, radius=1.5, shapeName=dyn_L_HairBang_01_jnt_capsule, center=<3 0 0.7>, rotation=<rot 0 90 90>}]
 dyn_L_HairBang_02_jnt [{shapeType=Capsule, length=2, radius=1.5, shapeName=dyn_L_HairBang_02_jnt_capsule, center=<3 0 0.7>, rotation=<rot 0 90 90>}]
 dyn_L_HairBang_03_jnt [{shapeType=Capsule, length=2, radius=1.5, shapeName=dyn_L_HairBang_03_jnt_capsule, center=<3 0 0.7>, rotation=<rot 0 90 90>}]
 dyn_L_HairBang_MSTR [{shapeType=Capsule, length=0, radius=0, shapeName=dyn_L_HairBang_MSTR_capsule}]
 dyn_L_HairTip_01_jnt [{shapeType=Capsule, length=2, radius=1.5, shapeName=dyn_L_HairTip_01_jnt_capsule, center=<3 0 0.7>, rotation=<rot 0 90 90>}]
 dyn_L_HairTip_02_jnt [{shapeType=Capsule, length=2, radius=1.5, shapeName=dyn_L_HairTip_02_jnt_capsule, center=<3 0 0.7>, rotation=<rot 0 90 90>}]
 dyn_L_Hairstrand_01_Tip_01_jnt [{
   shapeType=Capsule
   length=4
   radius=2
   shapeName=dyn_L_Hairstrand_01_Tip_01_jnt_capsule
   center=<4 0 1>
   rotation=<rot -3.4049 109.7201 99.4083>
  }]
 dyn_L_Hairstrand_01_Tip_02_jnt [{shapeType=Capsule, length=4, radius=2, shapeName=dyn_L_Hairstrand_01_Tip_02_jnt_capsule, center=<4 0 1>, rotation=<rot 0 90 90>}]
 dyn_L_Hairstrand_01_jnt [{shapeType=Capsule, length=4, radius=2, shapeName=dyn_L_Hairstrand_01_jnt_capsule, center=<4 0 1>, rotation=<rot 0 90 90>}]
 dyn_L_Hairstrand_02_Tip_01_jnt [{
   shapeType=Capsule
   length=4
   radius=2
   shapeName=dyn_L_Hairstrand_02_Tip_01_jnt_capsule
   center=<4 0 1>
   rotation=<rot -0.0001 90.0001 100.0001>
  }]
 dyn_L_Hairstrand_02_Tip_02_jnt [{shapeType=Capsule, length=4, radius=2, shapeName=dyn_L_Hairstrand_02_Tip_02_jnt_capsule, center=<4 0 1>, rotation=<rot 0 90 90>}]
 dyn_L_Hairstrand_02_jnt [{shapeType=Capsule, length=4, radius=2, shapeName=dyn_L_Hairstrand_02_jnt_capsule, center=<4 0 1>, rotation=<rot 0 90 90>}]
 dyn_L_Hairstrand_03_jnt [{shapeType=Capsule, length=4, radius=3, shapeName=dyn_L_Hairstrand_03_jnt_capsule, center=<4 -0.5 1>, rotation=<rot 0 90 90>}]
 dyn_L_Hairstrand_MSTR [{shapeType=Capsule, length=0, radius=0, shapeName=dyn_L_Hairstrand_MSTR_capsule}]
 dyn_R_HairBang_01_jnt [{shapeType=Capsule, length=2, radius=1.5, shapeName=dyn_R_HairBang_01_jnt_capsule, center=<3 0 0.7>, rotation=<rot 0 90 90>}]
 dyn_R_HairBang_02_jnt [{shapeType=Capsule, length=2, radius=1.5, shapeName=dyn_R_HairBang_02_jnt_capsule, center=<3 0 0.7>, rotation=<rot 0 90 90>}]
 dyn_R_HairBang_03_jnt [{shapeType=Capsule, length=2, radius=1.5, shapeName=dyn_R_HairBang_03_jnt_capsule, center=<3 0 0.7>, rotation=<rot 0 90 90>}]
 dyn_R_HairBang_04_jnt [{shapeType=Capsule, length=2, radius=1.5, shapeName=dyn_R_HairBang_04_jnt_capsule, center=<3 0 0.7>, rotation=<rot 0 90 90>}]
 dyn_R_HairBang_MSTR [{shapeType=Capsule, length=0, radius=0, shapeName=dyn_R_HairBang_MSTR_capsule}]
 dyn_R_Hairstrand_01_Tip_01_jnt [{shapeType=Capsule, length=4, radius=2, shapeName=dyn_R_Hairstrand_01_Tip_01_jnt_capsule, center=<4 0 1>, rotation=<rot 0 90 90>}]
 dyn_R_Hairstrand_01_Tip_02_jnt [{shapeType=Capsule, length=4, radius=2, shapeName=dyn_R_Hairstrand_01_Tip_02_jnt_capsule, center=<4 0 1>, rotation=<rot 0 90 90>}]
 dyn_R_Hairstrand_01_Tip_03_jnt [{shapeType=Capsule, length=4, radius=2, shapeName=dyn_R_Hairstrand_01_Tip_03_jnt_capsule, center=<4 0 1>, rotation=<rot 0 90 90>}]
 dyn_R_Hairstrand_01_jnt [{shapeType=Capsule, length=4, radius=2, shapeName=dyn_R_Hairstrand_01_jnt_capsule, center=<4 0 1>, rotation=<rot 0 90 90>}]
 dyn_R_Hairstrand_02_Tip_01_jnt [{shapeType=Capsule, length=4, radius=2, shapeName=dyn_R_Hairstrand_02_Tip_01_jnt_capsule, center=<4 0 1>, rotation=<rot 0 90 90>}]
 dyn_R_Hairstrand_02_Tip_02_jnt [{shapeType=Capsule, length=4, radius=2, shapeName=dyn_R_Hairstrand_02_Tip_02_jnt_capsule, center=<4 0 1>, rotation=<rot 0 90 90>}]
 dyn_R_Hairstrand_02_Tip_03_jnt [{shapeType=Capsule, length=4, radius=2, shapeName=dyn_R_Hairstrand_02_Tip_03_jnt_capsule, center=<4 0 1>, rotation=<rot 0 90 90>}]
 dyn_R_Hairstrand_02_jnt [{shapeType=Capsule, length=4, radius=2, shapeName=dyn_R_Hairstrand_02_jnt_capsule, center=<4 0 1>, rotation=<rot 0 90 90>}]
 dyn_R_Hairstrand_03_jnt [
  {shapeType=Capsule, length=4, radius=2, shapeName=dyn_R_Hairstrand_03_jnt_capsule, center=<4 0.4748 0.3485>, rotation=<rot 0 90 90>}
  {
   shapeType=Capsule
   length=4
   radius=2
   shapeName=dyn_R_Hairstrand_03_jnt_capsule
   center=<7.6185 2.2541 2.3696>
   rotation=<rot 0 89.9995 130.0003>
  }
 ]
 dyn_R_Hairstrand_MSTR [{shapeType=Capsule, length=0, radius=0, shapeName=dyn_R_Hairstrand_MSTR_capsule}]
 dyn_R_SmallHairBang_01_jnt [{shapeType=Capsule, length=2, radius=1.5, shapeName=dyn_R_SmallHairBang_01_jnt_capsule, center=<3 0 0.7>, rotation=<rot 0 90 90>}]
 dyn_R_SmallHairBang_02_jnt [{shapeType=Capsule, length=2, radius=1.5, shapeName=dyn_R_SmallHairBang_02_jnt_capsule, center=<3 0 0.7>, rotation=<rot 0 90 90>}]
 dyn_R_SmallHairBang_03_jnt [{shapeType=Capsule, length=2, radius=1.5, shapeName=dyn_R_SmallHairBang_03_jnt_capsule, center=<3 0 0.7>, rotation=<rot 0 90 90>}]
 dyn_R_SmallHairBang_04_jnt [{shapeType=Capsule, length=2, radius=1.5, shapeName=dyn_R_SmallHairBang_04_jnt_capsule, center=<3 0 0.7>, rotation=<rot 0 90 90>}]
 dyn_R_SmallHairBang_MSTR [{shapeType=Capsule, length=0, radius=0, shapeName=dyn_R_SmallHairBang_MSTR_capsule}]
 head [{shapeType=Capsule, length=6, radius=6.5, shapeName=head_capsule, center=<0.5788 1.8982 0>, rotation=<rot 0 90 90>}]
 neck_01 [{shapeType=Capsule, length=8, radius=4.5, shapeName=neck_01_capsule, center=<3.6683 -0.2876 0>, rotation=<rot 90 0 0>}]
 neck_02 [{shapeType=Capsule, length=6, radius=4.5, shapeName=neck_02_capsule, center=<2 0 0>, rotation=<rot 0 90 90>}]
 pelvis [{shapeType=Capsule, length=10, radius=10, shapeName=pelvis_capsule, center=<0 -2.3096 0>, rotation=<rot 0 -90 0>}]
 spine_01 [{shapeType=Capsule, length=8, radius=8, shapeName=spine_01_capsule, center=<0 -0.9904 0>, rotation=<rot 0 -90 0>}]
 spine_02 [{shapeType=Capsule, length=8, radius=8, shapeName=spine_02_capsule, rotation=<rot 0 -90 0>}]
 spine_03 [{shapeType=Capsule, length=15, radius=10, shapeName=spine_04_capsule, center=<0.229 2 0>, rotation=<rot 90 0 10.0001>}]
 spine_04 [{shapeType=Capsule, length=6, radius=6, shapeName=spine_04_capsule, rotation=<rot 0 90 90>}]
 spine_05 [
  {shapeType=Capsule, length=12, radius=9.25, shapeName=spine_05_capsule, center=<-2.5963 0.209 0>}
  {shapeType=Capsule, length=12, radius=8, shapeName=spine_05_capsule, center=<-5 4.5 0>}
 ]
 upperarm_l [{shapeType=Capsule, length=20, radius=3.5, shapeName=upperarm_l_capsule, center=<12 -0.5 0>, rotation=<rot 90 0 0>}]
 upperarm_r [{shapeType=Capsule, length=20, radius=3.5, shapeName=upperarm_r_capsule, center=<-12 0.5 0>, rotation=<rot 90 0 0>}]
}
constraints [
 {
  bone1Name=spine_01
  bone2Name=pelvis
  swing1LimitDegrees=45
  swing1Motion=Free
  swing2LimitDegrees=45
  swing2Motion=Free
  twistLimitDegrees=45
  twistMotion=Free
 }
 {
  bone1Name=spine_02
  bone2Name=spine_01
  swing1LimitDegrees=45
  swing1Motion=Free
  swing2LimitDegrees=45
  swing2Motion=Free
  twistLimitDegrees=45
  twistMotion=Free
 }
 {
  bone1Name=spine_03
  bone2Name=spine_02
  swing1LimitDegrees=45
  swing1Motion=Free
  swing2LimitDegrees=45
  swing2Motion=Free
  twistLimitDegrees=45
  twistMotion=Free
 }
 {
  bone1Name=spine_04
  bone2Name=spine_03
  swing1LimitDegrees=45
  swing1Motion=Free
  swing2LimitDegrees=45
  swing2Motion=Free
  twistLimitDegrees=45
  twistMotion=Free
 }
 {
  bone1Name=spine_05
  bone2Name=spine_04
  swing1LimitDegrees=45
  swing1Motion=Free
  swing2LimitDegrees=45
  swing2Motion=Free
  twistLimitDegrees=45
  twistMotion=Free
 }
 {
  bone1Name=neck_01
  bone2Name=spine_05
  swing1LimitDegrees=45
  swing1Motion=Free
  swing2LimitDegrees=45
  swing2Motion=Free
  twistLimitDegrees=45
  twistMotion=Free
 }
 {
  bone1Name=clavicle_l
  bone2Name=spine_05
  swing1LimitDegrees=45
  swing1Motion=Free
  swing2LimitDegrees=45
  swing2Motion=Free
  twistLimitDegrees=45
  twistMotion=Free
 }
 {
  bone1Name=clavicle_r
  bone2Name=spine_05
  swing1LimitDegrees=45
  swing1Motion=Free
  swing2LimitDegrees=45
  swing2Motion=Free
  twistLimitDegrees=45
  twistMotion=Free
 }
 {
  bone1Name=upperarm_l
  bone2Name=clavicle_l
  swing1LimitDegrees=45
  swing1Motion=Free
  swing2LimitDegrees=45
  swing2Motion=Free
  twistLimitDegrees=45
  twistMotion=Free
 }
 {
  bone1Name=upperarm_r
  bone2Name=clavicle_r
  swing1LimitDegrees=45
  swing1Motion=Free
  swing2LimitDegrees=45
  swing2Motion=Free
  twistLimitDegrees=45
  twistMotion=Free
 }
 {
  bone1Name=neck_02
  bone2Name=neck_01
  swing1LimitDegrees=45
  swing1Motion=Free
  swing2LimitDegrees=45
  swing2Motion=Free
  twistLimitDegrees=45
  twistMotion=Free
 }
 {
  bone1Name=head
  bone2Name=neck_02
  swing1LimitDegrees=45
  swing1Motion=Free
  swing2LimitDegrees=45
  swing2Motion=Free
  twistLimitDegrees=45
  twistMotion=Free
 }
 {
  bone1Name=dyn_L_HairBang_MSTR
  bone2Name=head
  swing1LimitDegrees=45
  swing1Motion=Free
  swing2LimitDegrees=45
  swing2Motion=Free
  twistLimitDegrees=45
  twistMotion=Free
 }
 {
  bone1Name=dyn_L_Hairstrand_MSTR
  bone2Name=head
  swing1LimitDegrees=45
  swing1Motion=Free
  swing2LimitDegrees=45
  swing2Motion=Free
  twistLimitDegrees=45
  twistMotion=Free
 }
 {
  bone1Name=dyn_R_HairBang_MSTR
  bone2Name=head
  swing1LimitDegrees=45
  swing1Motion=Free
  swing2LimitDegrees=45
  swing2Motion=Free
  twistLimitDegrees=45
  twistMotion=Free
 }
 {
  bone1Name=dyn_R_Hairstrand_MSTR
  bone2Name=head
  swing1LimitDegrees=45
  swing1Motion=Free
  swing2LimitDegrees=45
  swing2Motion=Free
  twistLimitDegrees=45
  twistMotion=Free
 }
 {
  bone1Name=dyn_R_SmallHairBang_MSTR
  bone2Name=head
  swing1LimitDegrees=45
  swing1Motion=Free
  swing2LimitDegrees=45
  swing2Motion=Free
  twistLimitDegrees=45
  twistMotion=Free
 }
 {bone1Name=dyn_L_HairBang_01_jnt, bone2Name=dyn_L_HairBang_MSTR, swing1LimitDegrees=5, swing2LimitDegrees=12, twistLimitDegrees=2}
 {bone1Name=dyn_L_HairBang_02_jnt, bone2Name=dyn_L_HairBang_01_jnt, swing1LimitDegrees=5, swing2LimitDegrees=12, twistLimitDegrees=2}
 {bone1Name=dyn_L_HairBang_03_jnt, bone2Name=dyn_L_HairBang_02_jnt, swing1LimitDegrees=5, swing2LimitDegrees=12, twistLimitDegrees=2}
 {bone1Name=dyn_L_HairTip_01_jnt, bone2Name=dyn_L_HairBang_03_jnt, swing1LimitDegrees=5, swing2LimitDegrees=12, twistLimitDegrees=2}
 {bone1Name=dyn_L_HairTip_02_jnt, bone2Name=dyn_L_HairBang_03_jnt, swing1LimitDegrees=5, swing2LimitDegrees=12, twistLimitDegrees=2}
 {bone1Name=dyn_L_Hairstrand_01_jnt, bone2Name=dyn_L_Hairstrand_MSTR, swing1LimitDegrees=5, swing2LimitDegrees=12, twistLimitDegrees=2}
 {bone1Name=dyn_L_Hairstrand_02_jnt, bone2Name=dyn_L_Hairstrand_01_jnt, swing1LimitDegrees=3, swing2LimitDegrees=12, twistLimitDegrees=1}
 {bone1Name=dyn_L_Hairstrand_03_jnt, bone2Name=dyn_L_Hairstrand_02_jnt, swing1LimitDegrees=12, swing2LimitDegrees=12, twistLimitDegrees=1}
 {
  bone1Name=dyn_L_Hairstrand_01_Tip_01_jnt
  bone2Name=dyn_L_Hairstrand_03_jnt
  swing1LimitDegrees=12
  swing2LimitDegrees=12
  twistLimitDegrees=2
 }
 {
  bone1Name=dyn_L_Hairstrand_02_Tip_01_jnt
  bone2Name=dyn_L_Hairstrand_03_jnt
  swing1LimitDegrees=12
  swing2LimitDegrees=12
  twistLimitDegrees=2
 }
 {
  bone1Name=dyn_L_Hairstrand_01_Tip_02_jnt
  bone2Name=dyn_L_Hairstrand_01_Tip_01_jnt
  swing1LimitDegrees=5
  swing2LimitDegrees=12
  twistLimitDegrees=2
 }
 {
  bone1Name=dyn_L_Hairstrand_02_Tip_02_jnt
  bone2Name=dyn_L_Hairstrand_02_Tip_01_jnt
  swing1LimitDegrees=5
  swing2LimitDegrees=12
  twistLimitDegrees=2
 }
 {bone1Name=dyn_R_HairBang_01_jnt, bone2Name=dyn_R_HairBang_MSTR, swing1LimitDegrees=5, swing2LimitDegrees=12, twistLimitDegrees=2}
 {bone1Name=dyn_R_HairBang_02_jnt, bone2Name=dyn_R_HairBang_01_jnt, swing1LimitDegrees=5, swing2LimitDegrees=12, twistLimitDegrees=2}
 {bone1Name=dyn_R_HairBang_03_jnt, bone2Name=dyn_R_HairBang_02_jnt, swing1LimitDegrees=5, swing2LimitDegrees=12, twistLimitDegrees=2}
 {bone1Name=dyn_R_HairBang_04_jnt, bone2Name=dyn_R_HairBang_03_jnt, swing1LimitDegrees=5, swing2LimitDegrees=12, twistLimitDegrees=2}
 {bone1Name=dyn_R_Hairstrand_01_jnt, bone2Name=dyn_R_Hairstrand_MSTR, swing1LimitDegrees=5, swing2LimitDegrees=12, twistLimitDegrees=2}
 {bone1Name=dyn_R_Hairstrand_02_jnt, bone2Name=dyn_R_Hairstrand_01_jnt, swing1LimitDegrees=3, swing2LimitDegrees=12, twistLimitDegrees=1}
 {bone1Name=dyn_R_Hairstrand_03_jnt, bone2Name=dyn_R_Hairstrand_02_jnt, swing1LimitDegrees=12, swing2LimitDegrees=12, twistLimitDegrees=1}
 {
  bone1Name=dyn_R_Hairstrand_01_Tip_01_jnt
  bone2Name=dyn_R_Hairstrand_03_jnt
  swing1LimitDegrees=12
  swing2LimitDegrees=12
  twistLimitDegrees=2
 }
 {
  bone1Name=dyn_R_Hairstrand_02_Tip_01_jnt
  bone2Name=dyn_R_Hairstrand_03_jnt
  swing1LimitDegrees=12
  swing2LimitDegrees=12
  twistLimitDegrees=2
 }
 {
  bone1Name=dyn_R_Hairstrand_01_Tip_02_jnt
  bone2Name=dyn_R_Hairstrand_01_Tip_01_jnt
  swing1LimitDegrees=5
  swing2LimitDegrees=12
  twistLimitDegrees=2
 }
 {
  bone1Name=dyn_R_Hairstrand_01_Tip_03_jnt
  bone2Name=dyn_R_Hairstrand_01_Tip_02_jnt
  swing1LimitDegrees=5
  swing2LimitDegrees=12
  twistLimitDegrees=2
 }
 {
  bone1Name=dyn_R_Hairstrand_02_Tip_02_jnt
  bone2Name=dyn_R_Hairstrand_02_Tip_01_jnt
  swing1LimitDegrees=5
  swing2LimitDegrees=12
  twistLimitDegrees=2
 }
 {
  bone1Name=dyn_R_Hairstrand_02_Tip_03_jnt
  bone2Name=dyn_R_Hairstrand_02_Tip_02_jnt
  swing1LimitDegrees=5
  swing2LimitDegrees=12
  twistLimitDegrees=2
 }
 {bone1Name=dyn_R_SmallHairBang_01_jnt, bone2Name=dyn_R_SmallHairBang_MSTR, swing1LimitDegrees=5, swing2LimitDegrees=12, twistLimitDegrees=2}
 {bone1Name=dyn_R_SmallHairBang_02_jnt, bone2Name=dyn_R_SmallHairBang_01_jnt, swing1LimitDegrees=5, swing2LimitDegrees=12, twistLimitDegrees=2}
 {bone1Name=dyn_R_SmallHairBang_03_jnt, bone2Name=dyn_R_SmallHairBang_02_jnt, swing1LimitDegrees=5, swing2LimitDegrees=12, twistLimitDegrees=2}
 {bone1Name=dyn_R_SmallHairBang_04_jnt, bone2Name=dyn_R_SmallHairBang_03_jnt, swing1LimitDegrees=5, swing2LimitDegrees=12, twistLimitDegrees=2}
 {
  bone1Name=dyn_Hair_bk_MSTR
  bone2Name=head
  swing1LimitDegrees=45
  swing1Motion=Free
  swing2LimitDegrees=45
  swing2Motion=Free
  twistLimitDegrees=45
  twistMotion=Free
 }
 {
  bone1Name=dyn_Hair_bk_01_ct
  bone2Name=dyn_Hair_bk_MSTR
  swing1LimitDegrees=45
  swing1Motion=Free
  swing2LimitDegrees=45
  swing2Motion=Free
  twistLimitDegrees=45
  twistMotion=Free
 }
 {
  bone1Name=dyn_Hair_bk_02_ct
  bone2Name=dyn_Hair_bk_01_ct
  swing1LimitDegrees=45
  swing1Motion=Free
  swing2LimitDegrees=45
  swing2Motion=Free
  twistLimitDegrees=45
  twistMotion=Free
 }
 {
  bone1Name=dyn_Hair_bk_03_ct
  bone2Name=dyn_Hair_bk_02_ct
  swing1LimitDegrees=45
  swing1Motion=Free
  swing2LimitDegrees=45
  swing2Motion=Free
  twistLimitDegrees=45
  twistMotion=Free
 }
 {
  bone1Name=dyn_Hair_bk_l_MSTR
  bone2Name=head
  swing1LimitDegrees=45
  swing1Motion=Free
  swing2LimitDegrees=45
  swing2Motion=Free
  twistLimitDegrees=45
  twistMotion=Free
 }
 {
  bone1Name=dyn_Hair_bk_01_l
  bone2Name=dyn_Hair_bk_l_MSTR
  swing1LimitDegrees=45
  swing1Motion=Free
  swing2LimitDegrees=45
  swing2Motion=Free
  twistLimitDegrees=45
  twistMotion=Free
 }
 {
  bone1Name=dyn_Hair_bk_02_l
  bone2Name=dyn_Hair_bk_01_l
  swing1LimitDegrees=45
  swing1Motion=Free
  swing2LimitDegrees=45
  swing2Motion=Free
  twistLimitDegrees=45
  twistMotion=Free
 }
 {
  bone1Name=dyn_Hair_bk_03_l
  bone2Name=dyn_Hair_bk_02_l
  swing1LimitDegrees=45
  swing1Motion=Free
  swing2LimitDegrees=45
  swing2Motion=Free
  twistLimitDegrees=45
  twistMotion=Free
 }
 {
  bone1Name=dyn_Hair_bk_r_MSTR
  bone2Name=head
  swing1LimitDegrees=45
  swing1Motion=Free
  swing2LimitDegrees=45
  swing2Motion=Free
  twistLimitDegrees=45
  twistMotion=Free
 }
 {
  bone1Name=dyn_Hair_bk_01_r
  bone2Name=dyn_Hair_bk_r_MSTR
  swing1LimitDegrees=45
  swing1Motion=Free
  swing2LimitDegrees=45
  swing2Motion=Free
  twistLimitDegrees=45
  twistMotion=Free
 }
 {
  bone1Name=dyn_Hair_bk_02_r
  bone2Name=dyn_Hair_bk_01_r
  swing1LimitDegrees=45
  swing1Motion=Free
  swing2LimitDegrees=45
  swing2Motion=Free
  twistLimitDegrees=45
  twistMotion=Free
 }
 {
  bone1Name=dyn_Hair_bk_03_r
  bone2Name=dyn_Hair_bk_02_r
  swing1LimitDegrees=45
  swing1Motion=Free
  swing2LimitDegrees=45
  swing2Motion=Free
  twistLimitDegrees=45
  twistMotion=Free
 }
]
